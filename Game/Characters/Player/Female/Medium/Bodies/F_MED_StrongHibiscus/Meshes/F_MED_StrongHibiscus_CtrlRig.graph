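# Control Rig for the F_MED_StrongHibiscus player body that adjusts shoulder cloth scrap bones as the upper arms move.
# Spherical pose readers on upperarm_l and upperarm_r output weights for Modify Transforms offsets on mid, front and back scraps.
# After the always-run chain, a branch on the Frontend bool (default false) adds back-scrap corrections on both sides.
# Those Frontend-only readers have debug drawing enabled.
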ControlRigBlueprint /Game/Characters/Player/Female/Medium/Bodies/F_MED_StrongHibiscus/Meshes/F_MED_StrongHibiscus_CtrlRig
graphs [
 RigVMFunctionLibrary {links [], nodes []}
 RigVMModel {nodes [
   RigUnit_BeginExecution = &RigUnit_BeginExecution()
   n1 = &RigUnit_SphericalPoseReader(
    ActiveRegionScaleFactors=(PositiveWidth=1,NegativeWidth=1,PositiveHeight=1,NegativeHeight=1)
    ActiveRegionSize=0.06
    Debug=(bDrawDebug=false,bDraw2D=False,bDrawLocalAxes=False,DebugScale=25,DebugSegments=20,DebugThickness=0.25)
    DriverAxis=(X=1,Y=0,Z=0)
    DriverItem=(Type=Bone,Name="upperarm_l")
    FalloffRegionScaleFactors=(PositiveWidth=1,NegativeWidth=1,PositiveHeight=1,NegativeHeight=1)
    FalloffSize=0.17
    FlipHeightScaling=False
    FlipWidthScaling=False
    OptionalParentItem=(Type=Bone,Name="None")
    RotationOffset=(X=45,Y=-30,Z=0)
    ExecutePin=@RigUnit_BeginExecution.ExecutePin
   )
   > &RigUnit_ModifyTransforms(
    ItemToModify=((Item=(Type=Bone,Name="dyn_l_cloth_scrap_mid_01"),Transform=(Rotation=(X=0.0784,Y=-0.0041,Z=-0.0522,W=0.9956),Translation=(X=-1,Y=0,Z=0),Scale3D=(X=1,Y=1,Z=1))))
    WeightMinimum=0
    Weight=@n1.OutputParam
   )
   > n2 = &RigUnit_SphericalPoseReader(
    ActiveRegionScaleFactors=(PositiveWidth=1,NegativeWidth=1,PositiveHeight=0.7,NegativeHeight=0.7)
    ActiveRegionSize=0.1
    Debug=(bDrawDebug=false,bDraw2D=False,bDrawLocalAxes=False,DebugScale=25,DebugSegments=20,DebugThickness=0.25)
    DriverAxis=(X=1,Y=0,Z=0)
    DriverItem=(Type=Bone,Name="upperarm_l")
    FalloffRegionScaleFactors=(PositiveWidth=1,NegativeWidth=1,PositiveHeight=0.7,NegativeHeight=0.7)
    FalloffSize=0.2
    FlipHeightScaling=False
    FlipWidthScaling=False
    OptionalParentItem=(Type=Bone,Name="None")
    RotationOffset=(X=50,Y=-40,Z=0)
   )
   > &RigUnit_ModifyTransforms(
    ItemToModify=((Item=(Type=Bone,Name="dyn_l_cloth_scrap_fr_a_01"),Transform=(Rotation=(X=0.0171,Y=-0.0036,Z=-0.2079,W=0.978),Translation=(X=-2,Y=0,Z=0),Scale3D=(X=1,Y=1,Z=1))))
    WeightMinimum=0
    Weight=@n2.OutputParam
   )
   > n3 = &RigUnit_SphericalPoseReader(
    ActiveRegionScaleFactors=(PositiveWidth=0.7,NegativeWidth=0.7,PositiveHeight=0.01,NegativeHeight=0.8)
    ActiveRegionSize=0.15
    Debug=(bDrawDebug=false,bDraw2D=False,bDrawLocalAxes=False,DebugScale=25,DebugSegments=20,DebugThickness=0.25)
    DriverAxis=(X=1,Y=0,Z=0)
    DriverItem=(Type=Bone,Name="upperarm_l")
    FalloffRegionScaleFactors=(PositiveWidth=0.7,NegativeWidth=0.7,PositiveHeight=1,NegativeHeight=1)
    FalloffSize=0.3
    FlipHeightScaling=False
    FlipWidthScaling=False
    OptionalParentItem=(Type=Bone,Name="None")
    RotationOffset=(X=75,Y=20,Z=-40)
   )
   > &RigUnit_ModifyTransforms(
    ItemToModify=((Item=(Type=Bone,Name="dyn_l_cloth_scrap_fr_b_01"),Transform=(Rotation=(X=-0.0619,Y=0.0322,Z=-0.4606,W=0.8849),Translation=(X=-1,Y=0,Z=0),Scale3D=(X=1,Y=1,Z=1))))
    WeightMinimum=0
    Weight=@n3.OutputParam
   )
   > n4 = &RigUnit_SphericalPoseReader(
    ActiveRegionScaleFactors=(PositiveWidth=0.2,NegativeWidth=0.3,PositiveHeight=0.1,NegativeHeight=0.7)
    ActiveRegionSize=0.2
    Debug=(bDrawDebug=false,bDraw2D=False,bDrawLocalAxes=False,DebugScale=25,DebugSegments=20,DebugThickness=0.25)
    DriverAxis=(X=1,Y=0,Z=0)
    DriverItem=(Type=Bone,Name="upperarm_l")
    FalloffRegionScaleFactors=(PositiveWidth=0.4,NegativeWidth=0.5,PositiveHeight=1,NegativeHeight=1)
    FalloffSize=0.25
    FlipHeightScaling=False
    FlipWidthScaling=False
    OptionalParentItem=(Type=Bone,Name="None")
    RotationOffset=(X=110,Y=30,Z=-50)
   )
   > &RigUnit_ModifyTransforms(
    ItemToModify=((Item=(Type=Bone,Name="dyn_l_cloth_scrap_fr_c_01"),Transform=(Rotation=(X=0,Y=0,Z=-0.6225,W=0.7826),Translation=(X=-1.5,Y=0,Z=0),Scale3D=(X=1,Y=1,Z=1))))
    WeightMinimum=0
    Weight=@n4.OutputParam
   )
   > n5 = &RigUnit_SphericalPoseReader(
    ActiveRegionScaleFactors=(PositiveWidth=1,NegativeWidth=1,PositiveHeight=1,NegativeHeight=1)
    ActiveRegionSize=0.1
    Debug=(bDrawDebug=false,bDraw2D=False,bDrawLocalAxes=False,DebugScale=25,DebugSegments=20,DebugThickness=0.25)
    DriverAxis=(X=1,Y=0,Z=0)
    DriverItem=(Type=Bone,Name="upperarm_l")
    FalloffRegionScaleFactors=(PositiveWidth=1,NegativeWidth=1,PositiveHeight=1,NegativeHeight=1)
    FalloffSize=0.2
    FlipHeightScaling=False
    FlipWidthScaling=False
    OptionalParentItem=(Type=Bone,Name="None")
    RotationOffset=(X=-20,Y=-40,Z=0)
   )
   > &RigUnit_ModifyTransforms(
    ItemToModify=((Item=(Type=Bone,Name="dyn_l_cloth_scrap_mid_01"),Transform=(Rotation=(X=-0.1218,Y=0.0032,Z=-0.026,W=0.9922),Translation=(X=-1.5,Y=-0.5,Z=0),Scale3D=(X=1,Y=1,Z=1))))
    WeightMinimum=0
    Weight=@n5.OutputParam
   )
   > &RigUnit_ModifyTransforms(
    ItemToModify=((Item=(Type=Bone,Name="dyn_l_cloth_scrap_bk_a_01"),Transform=(Rotation=(X=0.2411,Y=-0.0211,Z=-0.0487,W=0.969),Translation=(X=-2,Y=-0.7,Z=0),Scale3D=(X=1,Y=1,Z=1))))
    WeightMinimum=0
    Weight=@n5.OutputParam
   )
   > n6 = &RigUnit_SphericalPoseReader(
    ActiveRegionScaleFactors=(PositiveWidth=0.5,NegativeWidth=0.5,PositiveHeight=1,NegativeHeight=1)
    ActiveRegionSize=0.1
    Debug=(bDrawDebug=false,bDraw2D=False,bDrawLocalAxes=False,DebugScale=25,DebugSegments=20,DebugThickness=0.25)
    DriverAxis=(X=-1,Y=0,Z=0)
    DriverItem=(Type=Bone,Name="upperarm_r")
    FalloffRegionScaleFactors=(PositiveWidth=0.5,NegativeWidth=0.5,PositiveHeight=1,NegativeHeight=1)
    FalloffSize=0.4
    FlipHeightScaling=False
    FlipWidthScaling=False
    OptionalParentItem=(Type=Bone,Name="None")
    RotationOffset=(X=-125.4379,Y=32.7145,Z=25.7706)
   )
   > &RigUnit_ModifyTransforms(
    ItemToModify=((Item=(Type=Bone,Name="dyn_r_cloth_scrap_mid_01"),Transform=(Rotation=(X=-0.3574,Y=0.1721,Z=0.0059,W=0.918),Translation=(X=-1.2382,Y=-2.1989,Z=1.264),Scale3D=(X=1,Y=1,Z=1))))
    WeightMinimum=0
    Weight=@n6.OutputParam
   )
   > &RigUnit_ModifyTransforms(
    ItemToModify=((Item=(Type=Bone,Name="dyn_r_cloth_scrap_bk_a_01"),Transform=(Rotation=(X=0.0032,Y=0.1563,Z=0.2395,W=0.9582),Translation=(X=0,Y=0,Z=0),Scale3D=(X=1,Y=1,Z=1))))
    WeightMinimum=0
    Weight=@n6.OutputParam
   )
   > n7 = &RigUnit_SphericalPoseReader(
    ActiveRegionScaleFactors=(PositiveWidth=0.5,NegativeWidth=0.5,PositiveHeight=0.3,NegativeHeight=1)
    ActiveRegionSize=0.1
    Debug=(bDrawDebug=false,bDraw2D=False,bDrawLocalAxes=False,DebugScale=25,DebugSegments=20,DebugThickness=0.25)
    DriverAxis=(X=-1,Y=0,Z=0)
    DriverItem=(Type=Bone,Name="upperarm_r")
    FalloffRegionScaleFactors=(PositiveWidth=0.5,NegativeWidth=0.5,PositiveHeight=1.2,NegativeHeight=1)
    FalloffSize=0.35
    FlipHeightScaling=False
    FlipWidthScaling=False
    OptionalParentItem=(Type=Bone,Name="None")
    RotationOffset=(X=-81.6944,Y=42.157,Z=22.5118)
   )
   > &RigUnit_ModifyTransforms(
    ItemToModify=((Item=(Type=Bone,Name="dyn_r_cloth_scrap_fr_a_01"),Transform=(Rotation=(X=-0.049,Y=0.0008,Z=-0.5974,W=0.8004),Translation=(X=-3.56,Y=-1.7637,Z=1.5433),Scale3D=(X=1,Y=1,Z=1))))
    WeightMinimum=0
    Weight=@n7.OutputParam
   )
   > n8 = &RigUnit_SphericalPoseReader(
    ActiveRegionScaleFactors=(PositiveWidth=0.5,NegativeWidth=0.5,PositiveHeight=1,NegativeHeight=1.5)
    ActiveRegionSize=0.1
    Debug=(bDrawDebug=false,bDraw2D=False,bDrawLocalAxes=False,DebugScale=25,DebugSegments=20,DebugThickness=0.25)
    DriverAxis=(X=-1,Y=0,Z=0)
    DriverItem=(Type=Bone,Name="upperarm_r")
    FalloffRegionScaleFactors=(PositiveWidth=0.5,NegativeWidth=0.5,PositiveHeight=1.2,NegativeHeight=1.5)
    FalloffSize=0.35
    FlipHeightScaling=False
    FlipWidthScaling=False
    OptionalParentItem=(Type=Bone,Name="None")
    RotationOffset=(X=-70,Y=40,Z=20)
   )
   > n9 = &RigUnit_ModifyTransforms(
    ItemToModify=((Item=(Type=Bone,Name="dyn_r_cloth_scrap_fr_b_01"),Transform=(Rotation=(X=0.0273,Y=-0.0703,Z=-0.8347,W=0.5455),Translation=(X=-1.7722,Y=1.1426,Z=0.6887),Scale3D=(X=1,Y=1,Z=1))))
    WeightMinimum=0
    Weight=@n8.OutputParam
   )
   RigVMFunction_ControlFlowBranch = &RigVMFunction_ControlFlowBranch(ExecuteContext=@n9.ExecutePin, Condition=RigVMVariableNode() {variable=Frontend}.Value)
   n10 = &RigUnit_SphericalPoseReader(
    ActiveRegionScaleFactors=(PositiveWidth=0.5,NegativeWidth=0.5,PositiveHeight=0.5,NegativeHeight=1)
    ActiveRegionSize=0.1
    Debug=(bDrawDebug=true,bDraw2D=False,bDrawLocalAxes=False,DebugScale=25,DebugSegments=20,DebugThickness=0.25)
    DriverAxis=(X=-1,Y=0,Z=0)
    DriverItem=(Type=Bone,Name="upperarm_r")
    FalloffRegionScaleFactors=(PositiveWidth=0.3,NegativeWidth=0.3,PositiveHeight=0.2,NegativeHeight=0.7)
    FalloffSize=0.35
    FlipHeightScaling=False
    FlipWidthScaling=False
    OptionalParentItem=(Type=Bone,Name="None")
    RotationOffset=(X=78.1195,Y=82.3868,Z=26.8621)
    ExecutePin=@RigVMFunction_ControlFlowBranch.True
   )
   > &RigUnit_ModifyTransforms(
    ItemToModify=((Item=(Type=Bone,Name="dyn_r_cloth_scrap_bk_a_01"),Transform=(Rotation=(X=0.0903,Y=-0.1439,Z=-0.1178,W=0.9784),Translation=(X=-0.7342,Y=-1.4292,Z=-0.2001),Scale3D=(X=1,Y=1,Z=1))))
    WeightMinimum=0
    Weight=@n10.OutputParam
   )
   > &RigUnit_ModifyTransforms(
    ItemToModify=((Item=(Type=Bone,Name="dyn_r_cloth_scrap_bk_b_01"),Transform=(Rotation=(X=0.0925,Y=-0.2843,Z=-0.3752,W=0.8774),Translation=(X=-1.6064,Y=-1.9742,Z=-0.9001),Scale3D=(X=1,Y=1,Z=1))))
    WeightMinimum=0
    Weight=@n10.OutputParam
   )
   > n11 = &RigUnit_SphericalPoseReader(
    ActiveRegionScaleFactors=(PositiveWidth=1,NegativeWidth=1,PositiveHeight=1,NegativeHeight=1)
    ActiveRegionSize=0.04
    Debug=(bDrawDebug=true,bDraw2D=False,bDrawLocalAxes=False,DebugScale=25,DebugSegments=20,DebugThickness=0.25)
    DriverAxis=(X=-1,Y=0,Z=0)
    DriverItem=(Type=Bone,Name="upperarm_r")
    FalloffRegionScaleFactors=(PositiveWidth=1,NegativeWidth=1,PositiveHeight=1,NegativeHeight=1)
    FalloffSize=0.12
    FlipHeightScaling=False
    FlipWidthScaling=False
    OptionalParentItem=(Type=Bone,Name="None")
    RotationOffset=(X=23.29,Y=60.4752,Z=10.6447)
   )
   > &RigUnit_ModifyTransforms(
    ItemToModify=((Item=(Type=Bone,Name="dyn_r_cloth_scrap_bk_a_01"),Transform=(Rotation=(X=0.1348,Y=-0.0499,Z=-0.0068,W=0.9896),Translation=(X=-0.4554,Y=-0.749,Z=-0.2079),Scale3D=(X=1,Y=1,Z=1))))
    WeightMinimum=0
    Weight=@n11.OutputParam
   )
   > &RigUnit_ModifyTransforms(
    ItemToModify=((Item=(Type=Bone,Name="dyn_r_cloth_scrap_bk_b_01"),Transform=(Rotation=(X=0.0461,Y=-0.1886,Z=-0.0496,W=0.9797),Translation=(X=-0.6875,Y=-1.3833,Z=-1.3684),Scale3D=(X=1,Y=1,Z=1))))
    WeightMinimum=0
    Weight=@n11.OutputParam
   )
   > n12 = &RigUnit_SphericalPoseReader(
    ActiveRegionScaleFactors=(PositiveWidth=0.5,NegativeWidth=0.5,PositiveHeight=0.5,NegativeHeight=1)
    ActiveRegionSize=0.1
    Debug=(bDrawDebug=true,bDraw2D=False,bDrawLocalAxes=False,DebugScale=25,DebugSegments=20,DebugThickness=0.25)
    DriverAxis=(X=1,Y=0,Z=0)
    DriverItem=(Type=Bone,Name="upperarm_l")
    FalloffRegionScaleFactors=(PositiveWidth=0.3,NegativeWidth=0.3,PositiveHeight=0.2,NegativeHeight=0.7)
    FalloffSize=0.35
    FlipHeightScaling=False
    FlipWidthScaling=False
    OptionalParentItem=(Type=Bone,Name="None")
    RotationOffset=(X=78,Y=100,Z=222)
   )
   > &RigUnit_ModifyTransforms(
    ItemToModify=((Item=(Type=Bone,Name="dyn_l_cloth_scrap_bk_a_01"),Transform=(Rotation=(X=-0.0585,Y=0,Z=0,W=0.9983),Translation=(X=0,Y=-0.6717,Z=0),Scale3D=(X=1,Y=1,Z=1))))
    WeightMinimum=0
    Weight=@n12.OutputParam
   )
   > &RigUnit_ModifyTransforms(
    ItemToModify=((Item=(Type=Bone,Name="dyn_l_cloth_scrap_bk_b_01"),Transform=(Rotation=(X=-0.1866,Y=0.0397,Z=-0.2041,W=0.9602),Translation=(X=-0.3,Y=-0.7,Z=0.3),Scale3D=(X=1,Y=1,Z=1))))
    WeightMinimum=0
    Weight=@n12.OutputParam
   )
   > n13 = &RigUnit_SphericalPoseReader(
    ActiveRegionScaleFactors=(PositiveWidth=1,NegativeWidth=1,PositiveHeight=1,NegativeHeight=1)
    ActiveRegionSize=0.04
    Debug=(bDrawDebug=true,bDraw2D=False,bDrawLocalAxes=False,DebugScale=25,DebugSegments=20,DebugThickness=0.25)
    DriverAxis=(X=1,Y=0,Z=0)
    DriverItem=(Type=Bone,Name="upperarm_l")
    FalloffRegionScaleFactors=(PositiveWidth=1,NegativeWidth=1,PositiveHeight=1,NegativeHeight=1)
    FalloffSize=0.12
    FlipHeightScaling=False
    FlipWidthScaling=False
    OptionalParentItem=(Type=Bone,Name="None")
    RotationOffset=(X=0,Y=-122,Z=-15)
   )
   > &RigUnit_ModifyTransforms(
    ItemToModify=((Item=(Type=Bone,Name="dyn_l_cloth_scrap_bk_a_01"),Transform=(Rotation=(X=-0.1231,Y=0.1248,Z=-0.0449,W=0.9835),Translation=(X=-1.2446,Y=-0.9182,Z=0),Scale3D=(X=1,Y=1,Z=1))))
    WeightMinimum=0
    Weight=@n13.OutputParam
   )
   > &RigUnit_ModifyTransforms(
    ItemToModify=((Item=(Type=Bone,Name="dyn_l_cloth_scrap_bk_b_01"),Transform=(Rotation=(X=0.0229,Y=0.027,Z=-0.1354,W=0.9902),Translation=(X=-1.3968,Y=-1.1541,Z=0.1686),Scale3D=(X=1,Y=1,Z=1))))
    WeightMinimum=0
    Weight=@n13.OutputParam
   )
  ]}
]
hierarchy [
 root {loc=<0 0 0>, rot=<0 0 0 1>}
 pelvis {parent=root, loc=<0 -0.5301 98.0571>, rot=<0.0127 0.707 -0.0127 -0.707>}
 spine_01 {parent=pelvis, loc=<6.1208 -0.9328 0>, rot=<0 0 0.034 -0.9994>}
 spine_02 {parent=spine_01, loc=<6.5996 -0.2095 0>, rot=<0 0 0.0238 -0.9997>}
 spine_03 {parent=spine_02, loc=<6.4914 -0.1237 0>, rot=<0 0 0.0149 -0.9999>}
 spine_04 {parent=spine_03, loc=<6.9598 0.0794 0>, rot=<0 0 -0.0132 -0.9999>}
 spine_05 {parent=spine_04, loc=<6.9987 0 0>, rot=<0 0 -0.0015 -1>}
 attach_cape {parent=spine_05, loc=<9.3216 -8.4501 0>, rot=<0.0616 -0.7044 -0.0616 -0.7044>}
 neck_01 {parent=spine_05, loc=<11.3138 0.1529 0>, rot=<0 0 -0.2003 -0.9797>}
 neck_02 {parent=neck_01, loc=<4.5627 0 0>, rot=<0 0 0.0504 -0.9987>}
 head {parent=neck_02, loc=<5.3914 0 0>, rot=<0 0 0.115 -0.9934>}
 attach_backpack {parent=spine_05, loc=<0 -8.35 0>, rot=<-0.0093 -0.707 0.0093 -0.707>}
 clavicle_l {parent=spine_05, loc=<5.4831 -0.1053 -3.2946>, rot=<0.0465 -0.7289 -0.009 -0.6829>}
 upperarm_l {parent=clavicle_l, loc=<9.8658 -1.2136 -0.4105>, rot=<0.0164 -0.3245 -0.0269 -0.9454>}
 lowerarm_l {parent=upperarm_l, loc=<26.465 -0.8345 0.259>, rot=<0.0036 0.0161 -0.2298 -0.9731>}
 hand_l {parent=lowerarm_l, loc=<22.7355 -0.3203 -0.6105>, rot=<-0.6784 -0.0223 0.0735 -0.7307>}
 thumb_01_l {parent=hand_l, loc=<1.3118 -0.3191 -2.459>, rot=<0.6546 -0.4042 0.0664 -0.6354>}
 thumb_02_l {parent=thumb_01_l, loc=<5.9336 -0.0448 0.0401>, rot=<-0.0037 0.0213 0.298 -0.9543>}
 thumb_03_l {parent=thumb_02_l, loc=<2.2223 0.3682 -0.0907>, rot=<0.0118 -0.013 0.0066 -0.9998>}
 index_metacarpal_l {parent=hand_l, loc=<1.6535 0.0183 -2.96>, rot=<0.0158 -0.0124 -0.0173 -0.9996>}
 index_01_l {parent=index_metacarpal_l, loc=<8.3008 0.0357 -0.0524>, rot=<-0.0044 0.0485 0.0924 -0.9945>}
 index_02_l {parent=index_01_l, loc=<3.3587 -0.3004 0.0452>, rot=<-0.002 0.0027 0.3225 -0.9466>}
 index_03_l {parent=index_02_l, loc=<2.5578 -0.123 -0.0252>, rot=<-0.002 0.0067 0.2558 -0.9667>}
 middle_metacarpal_l {parent=hand_l, loc=<2.1856 -0.1725 -0.5984>, rot=<0.006 0.0168 -0.0329 -0.9993>}
 middle_01_l {parent=middle_metacarpal_l, loc=<7.3275 -0.2447 0.0129>, rot=<-0.0208 0.0042 0.0858 -0.9961>}
 middle_02_l {parent=middle_01_l, loc=<4.0123 0.1401 0.1175>, rot=<-0.0218 0.0258 0.3512 -0.9357>}
 middle_03_l {parent=middle_02_l, loc=<2.6077 0.0534 0.1373>, rot=<0.0082 0.0188 0.1708 -0.9851>}
 ring_metacarpal_l {parent=hand_l, loc=<2.2074 -0.3625 1.3638>, rot=<-0.0284 0.0722 -0.0458 -0.9959>}
 ring_01_l {parent=ring_metacarpal_l, loc=<7.0295 -0.6197 -0.1446>, rot=<-0.0489 -0.0252 0.1892 -0.9804>}
 ring_02_l {parent=ring_01_l, loc=<3.6174 0.4647 -0.1027>, rot=<0 0 0.2806 -0.9598>}
 ring_03_l {parent=ring_02_l, loc=<2.4152 0.114 0.0199>, rot=<0 0 0.1189 -0.9929>}
 pinky_metacarpal_l {parent=hand_l, loc=<1.7616 -0.9753 3.3814>, rot=<-0.0397 0.0814 -0.069 -0.9935>}
 pinky_01_l {parent=pinky_metacarpal_l, loc=<7.1709 -0.4831 -0.1094>, rot=<-0.0654 -0.0035 0.2242 -0.9723>}
 pinky_02_l {parent=pinky_01_l, loc=<2.8274 0.0893 -0.1043>, rot=<0.0037 0.0011 0.2854 -0.9584>}
 pinky_03_l {parent=pinky_02_l, loc=<2.2046 0.0307 -0.0078>, rot=<-0.0002 0.0098 0.1445 -0.9895>}
 lowerarm_twist_02_l {parent=lowerarm_l, loc=<10.6707 0 0>, rot=<0 0 0 1>}
 lowerarm_twist_01_l {parent=lowerarm_l, loc=<18.316 0 0>, rot=<0 0 0 1>}
 deform_elbow_in_base_l {parent=lowerarm_l, loc=<0.1834 3.6062 -0.2572>, rot=<0 0 0 -1>}
 deform_elbow_in_l {parent=deform_elbow_in_base_l, loc=<0 0 0>, rot=<0 0 0 1>}
 deform_elbow_out_base_l {parent=lowerarm_l, loc=<0 -4 0>, rot=<0 0 0 1>}
 deform_elbow_out_l {parent=deform_elbow_out_base_l, loc=<0 0 0>, rot=<0 0 0 1>}
 upperarm_twist_01_l {parent=upperarm_l, loc=<7.2135 -0.1134 0>, rot=<0 0 0 1>}
 deform_upperarm_bk_base_l {parent=upperarm_twist_01_l, loc=<-6.4053 -4.3746 0.314>, rot=<0 0 0 -1>}
 deform_upperarm_bk_l {parent=deform_upperarm_bk_base_l, loc=<0 0 0>, rot=<0 0 0 1>}
 deform_upperarm_fr_base_l {parent=upperarm_twist_01_l, loc=<-5.6312 5.092 0.3354>, rot=<0 0 0 -1>}
 deform_upperarm_fr_l {parent=deform_upperarm_fr_base_l, loc=<0 0 0>, rot=<0 0 0 1>}
 deform_upperarm_up_base_l {parent=upperarm_twist_01_l, loc=<-8.891 0.4825 3.9517>, rot=<0 0 0 -1>}
 deform_upperarm_up_l {parent=deform_upperarm_up_base_l, loc=<0 0 0>, rot=<0 0 0 1>}
 upperarm_twist_02_l {parent=upperarm_l, loc=<19.7998 -0.1134 0>, rot=<0 0 0 1>}
 dyn_l_cloth_scrap_mid_MSTR {parent=clavicle_l, loc=<8.272 -2.527 4.4106>, rot=<0.0008 0.0336 -0.0252 -0.9991>}
 dyn_l_cloth_scrap_mid_01 {parent=dyn_l_cloth_scrap_mid_MSTR, loc=<0 0 0>, rot=<0.6807 0.1913 0.1913 -0.6807>}
 dyn_l_cloth_scrap_fr_a_MSTR {parent=clavicle_l, loc=<9.1002 3.5382 1.6055>, rot=<0.0008 0.0336 -0.0252 -0.9991>}
 dyn_l_cloth_scrap_fr_a_01 {parent=dyn_l_cloth_scrap_fr_a_MSTR, loc=<0 0 0>, rot=<0.9006 0.2459 0.1122 -0.3403>}
 dyn_l_cloth_scrap_fr_b_MSTR {parent=clavicle_l, loc=<8.649 5.4508 -1.4693>, rot=<0.0008 0.0336 -0.0252 -0.9991>}
 dyn_l_cloth_scrap_fr_b_01 {parent=dyn_l_cloth_scrap_fr_b_MSTR, loc=<0 0 0>, rot=<0.9781 0.1271 0.0367 -0.1607>}
 dyn_l_cloth_scrap_fr_c_MSTR {parent=clavicle_l, loc=<9.6348 6.2348 -5.9815>, rot=<0.0008 0.0336 -0.0252 -0.9991>}
 dyn_l_cloth_scrap_fr_c_01 {parent=dyn_l_cloth_scrap_fr_c_MSTR, loc=<0 0 0>, rot=<0.9847 -0.0865 0.1053 0.1088>}
 dyn_l_cloth_scrap_bk_a_MSTR {parent=clavicle_l, loc=<9.4127 -7.9465 -0.4773>, rot=<0.0008 0.0336 -0.0252 -0.9991>}
 dyn_l_cloth_scrap_bk_a_01 {parent=dyn_l_cloth_scrap_bk_a_MSTR, loc=<0 0 0>, rot=<0 0 0.1366 -0.9906>}
 dyn_l_cloth_scrap_bk_b_MSTR {parent=clavicle_l, loc=<9.7416 -8.1502 -5.2153>, rot=<0.0008 0.0336 -0.0252 -0.9991>}
 dyn_l_cloth_scrap_bk_b_01 {parent=dyn_l_cloth_scrap_bk_b_MSTR, loc=<0 0 0>, rot=<0 0 0.0845 -0.9964>}
 clavicle_r {parent=spine_05, loc=<5.4831 -0.1053 3.2946>, rot=<0.7289 0.0465 -0.6829 0.009>}
 upperarm_r {parent=clavicle_r, loc=<-9.8658 1.2136 0.4105>, rot=<0.0164 -0.3245 -0.0269 -0.9454>}
 lowerarm_r {parent=upperarm_r, loc=<-26.465 0.8345 -0.259>, rot=<0.0036 0.0161 -0.2298 -0.9731>}
 hand_r {parent=lowerarm_r, loc=<-22.7355 0.3203 0.6106>, rot=<-0.6784 -0.0223 0.0735 -0.7307>}
 thumb_01_r {parent=hand_r, loc=<-1.3118 0.3191 2.459>, rot=<0.6546 -0.4042 0.0664 -0.6354>}
 thumb_02_r {parent=thumb_01_r, loc=<-5.9336 0.0448 -0.0401>, rot=<-0.0037 0.0213 0.298 -0.9543>}
 thumb_03_r {parent=thumb_02_r, loc=<-2.2223 -0.3682 0.0907>, rot=<0.0118 -0.013 0.0066 -0.9998>}
 index_metacarpal_r {parent=hand_r, loc=<-1.6535 -0.0183 2.96>, rot=<0.0158 -0.0124 -0.0173 -0.9996>}
 index_01_r {parent=index_metacarpal_r, loc=<-8.3008 -0.0357 0.0524>, rot=<-0.0044 0.0485 0.0924 -0.9945>}
 index_02_r {parent=index_01_r, loc=<-3.3587 0.3004 -0.0452>, rot=<-0.002 0.0027 0.3225 -0.9466>}
 index_03_r {parent=index_02_r, loc=<-2.5578 0.123 0.0252>, rot=<-0.002 0.0067 0.2558 -0.9667>}
 middle_metacarpal_r {parent=hand_r, loc=<-2.1856 0.1725 0.5984>, rot=<0.006 0.0168 -0.0329 -0.9993>}
 middle_01_r {parent=middle_metacarpal_r, loc=<-7.3275 0.2447 -0.0129>, rot=<-0.0208 0.0042 0.0858 -0.9961>}
 middle_02_r {parent=middle_01_r, loc=<-4.0123 -0.1401 -0.1175>, rot=<-0.0218 0.0258 0.3512 -0.9357>}
 middle_03_r {parent=middle_02_r, loc=<-2.6077 -0.0534 -0.1373>, rot=<0.0082 0.0188 0.1708 -0.9851>}
 ring_metacarpal_r {parent=hand_r, loc=<-2.2074 0.3625 -1.3638>, rot=<-0.0284 0.0722 -0.0458 -0.9959>}
 ring_01_r {parent=ring_metacarpal_r, loc=<-7.0295 0.6197 0.1446>, rot=<-0.0489 -0.0252 0.1892 -0.9804>}
 ring_02_r {parent=ring_01_r, loc=<-3.6174 -0.4647 0.1027>, rot=<0 0 0.2806 -0.9598>}
 ring_03_r {parent=ring_02_r, loc=<-2.4152 -0.114 -0.0199>, rot=<0 0 0.1189 -0.9929>}
 pinky_metacarpal_r {parent=hand_r, loc=<-1.7616 0.9753 -3.3814>, rot=<-0.0397 0.0814 -0.069 -0.9935>}
 pinky_01_r {parent=pinky_metacarpal_r, loc=<-7.1709 0.4831 0.1094>, rot=<-0.0654 -0.0035 0.2242 -0.9723>}
 pinky_02_r {parent=pinky_01_r, loc=<-2.8274 -0.0893 0.1043>, rot=<0.0037 0.0011 0.2854 -0.9584>}
 pinky_03_r {parent=pinky_02_r, loc=<-2.2046 -0.0307 0.0078>, rot=<-0.0002 0.0098 0.1445 -0.9895>}
 lowerarm_twist_02_r {parent=lowerarm_r, loc=<-10.6707 0 0>, rot=<0 0 0 -1>}
 lowerarm_twist_01_r {parent=lowerarm_r, loc=<-18.316 0 0>, rot=<0 0 0 -1>}
 deform_elbow_in_base_r {parent=lowerarm_r, loc=<-0.1834 -3.6062 0.2572>, rot=<0 0 0 -1>}
 deform_elbow_in_r {parent=deform_elbow_in_base_r, loc=<0 0 0>, rot=<0 0 0 1>}
 deform_elbow_out_base_r {parent=lowerarm_r, loc=<0 4 -0.0001>, rot=<0 0 0 -1>}
 deform_elbow_out_r {parent=deform_elbow_out_base_r, loc=<0 0 0>, rot=<0 0 0 1>}
 upperarm_twist_01_r {parent=upperarm_r, loc=<-7.2135 0.1134 0>, rot=<0 0 0 -1>}
 deform_upperarm_bk_base_r {parent=upperarm_twist_01_r, loc=<6.4053 4.3746 -0.314>, rot=<0 0 0 -1>}
 deform_upperarm_bk_r {parent=deform_upperarm_bk_base_r, loc=<0 0 0>, rot=<0 0 0 1>}
 deform_upperarm_fr_base_r {parent=upperarm_twist_01_r, loc=<5.6312 -5.092 -0.3354>, rot=<0 0 0 -1>}
 deform_upperarm_fr_r {parent=deform_upperarm_fr_base_r, loc=<0 0 0>, rot=<0 0 0 1>}
 deform_upperarm_up_base_r {parent=upperarm_twist_01_r, loc=<8.891 -0.4825 -3.9517>, rot=<0 0 0 -1>}
 deform_upperarm_up_r {parent=deform_upperarm_up_base_r, loc=<0 0 0>, rot=<0 0 0 1>}
 upperarm_twist_02_r {parent=upperarm_r, loc=<-19.7998 0.1134 0>, rot=<0 0 0 -1>}
 dyn_r_cloth_scrap_mid_MSTR {parent=clavicle_r, loc=<-8.3084 2.2601 -4.2918>, rot=<0.0008 0.0336 -0.0252 -0.9991>}
 dyn_r_cloth_scrap_mid_01 {parent=dyn_r_cloth_scrap_mid_MSTR, loc=<0 0 0>, rot=<-0.1913 0.6807 0.6807 0.1913>}
 dyn_r_cloth_scrap_fr_a_MSTR {parent=clavicle_r, loc=<-9.0524 -3.5357 -1.6019>, rot=<0.0008 0.0336 -0.0252 -0.9991>}
 dyn_r_cloth_scrap_fr_a_01 {parent=dyn_r_cloth_scrap_fr_a_MSTR, loc=<0 0 0>, rot=<0.229 -0.8941 -0.3518 -0.1558>}
 dyn_r_cloth_scrap_fr_b_MSTR {parent=clavicle_r, loc=<-8.657 -5.8685 1.9504>, rot=<0.0008 0.0336 -0.0252 -0.9991>}
 dyn_r_cloth_scrap_fr_b_01 {parent=dyn_r_cloth_scrap_fr_b_MSTR, loc=<0 0 0>, rot=<0.0512 -0.9757 -0.1903 -0.0963>}
 dyn_r_cloth_scrap_fr_c_MSTR {parent=clavicle_r, loc=<-9.3015 -6.2431 6.3422>, rot=<0.0008 0.0336 -0.0252 -0.9991>}
 dyn_r_cloth_scrap_fr_c_01 {parent=dyn_r_cloth_scrap_fr_c_MSTR, loc=<0 0 0>, rot=<-0.1554 -0.9724 0.1005 -0.1421>}
 dyn_r_cloth_scrap_bk_a_MSTR {parent=clavicle_r, loc=<-9.4088 7.9489 1.133>, rot=<0.0008 0.0336 -0.0252 -0.9991>}
 dyn_r_cloth_scrap_bk_a_01 {parent=dyn_r_cloth_scrap_bk_a_MSTR, loc=<0 0 0>, rot=<-0.1107 0.1126 0.9765 0.147>}
 dyn_r_cloth_scrap_bk_b_MSTR {parent=clavicle_r, loc=<-9.746 8.1525 5.9937>, rot=<0.0008 0.0336 -0.0252 -0.9991>}
 dyn_r_cloth_scrap_bk_b_01 {parent=dyn_r_cloth_scrap_bk_b_MSTR, loc=<0 0 0>, rot=<0 0 0.9964 0.0845>}
 dyn_knot_straps_MSTR {parent=spine_02, loc=<0.3179 10.21 -1.1609>, rot=<0.0282 -0.7065 -0.0282 -0.7065>}
 dyn_knot_strap_a_01 {parent=dyn_knot_straps_MSTR, loc=<0.8799 2.2076 -2.4243>, rot=<0.8561 0.2242 -0.3792 -0.2703>}
 dyn_knot_strap_a_02 {parent=dyn_knot_strap_a_01, loc=<6.5375 0 0>, rot=<-0.0252 -0.1539 -0.1595 -0.9748>}
 dyn_knot_strap_b_01 {parent=dyn_knot_straps_MSTR, loc=<-0.0519 1.8725 -3.7102>, rot=<0.7947 0.0715 -0.5664 -0.2061>}
 dyn_knot_strap_b_02 {parent=dyn_knot_strap_b_01, loc=<5.8973 0 0>, rot=<0.1329 -0.167 0.0382 -0.9762>}
 thigh_l {parent=pelvis, loc=<-1.0855 -0.2109 -8.7053>, rot=<0.0009 0.0355 0.0155 -0.9993>}
 calf_l {parent=thigh_l, loc=<-42.5284 -0.0019 -0.7297>, rot=<0.0006 -0.0161 -0.0391 -0.9991>}
 foot_l {parent=calf_l, loc=<-46.7401 -2.5014 -0.592>, rot=<0.0427 -0.0238 0.0353 -0.9982>}
 ball_l {parent=foot_l, loc=<-3.5658 10.8066 -0.1544>, rot=<0.0001 -0.0001 -0.7063 -0.7079>}
 calf_twist_02_l {parent=calf_l, loc=<-19.1546 -1.7943 -0.2638>, rot=<0.0237 0.0095 -0.0035 -0.9997>}
 calf_twist_01_l {parent=calf_l, loc=<-35.8483 -1.9197 -0.5794>, rot=<0.0237 0.0095 -0.0035 -0.9997>}
 dyn_l_shoelaces_MSTR {parent=calf_l, loc=<-15.16 2.5622 -1.5408>, rot=<-0.0309 -0.7201 0.0276 -0.6927>}
 dyn_l_shoelace_a_01 {parent=dyn_l_shoelaces_MSTR, loc=<-0.7832 0.7109 -1.0009>, rot=<0.6794 0.1959 -0.6794 0.1959>}
 dyn_l_shoelace_b_01 {parent=dyn_l_shoelaces_MSTR, loc=<1.1702 0.3507 -1.037>, rot=<0.6493 -0.2799 -0.6493 -0.2799>}
 dyn_l_shoelace_c_01 {parent=dyn_l_shoelaces_MSTR, loc=<0.1667 0.0079 -1.3356>, rot=<0.7071 0.0059 -0.7071 0.0059>}
 dyn_l_shoelace_c_02 {parent=dyn_l_shoelace_c_01, loc=<6.2976 -0.3419 -0.3995>, rot=<0 0 0 1>}
 dyn_l_shoelace_d_01 {parent=dyn_l_shoelaces_MSTR, loc=<0.6352 -0.0045 -1.3356>, rot=<0.7071 0.0059 -0.7071 0.0059>}
 dyn_l_shoelace_d_02 {parent=dyn_l_shoelace_d_01, loc=<6.2023 -0.5954 -1.2368>, rot=<0 0 0 1>}
 deform_knee_in_base_l {parent=calf_l, loc=<0.8031 -5.0231 0.2456>, rot=<0 0 0 -1>}
 deform_knee_in_l {parent=deform_knee_in_base_l, loc=<0 0 0>, rot=<0 0 0 1>}
 deform_knee_out_base_l {parent=calf_l, loc=<0 4 0>, rot=<0 0 0 -1>}
 deform_knee_out_l {parent=deform_knee_out_base_l, loc=<0 0 0>, rot=<0 0 0 1>}
 thigh_twist_01_l {parent=thigh_l, loc=<-27.952 0 0>, rot=<0.0013 0 0 -1>}
 deform_groin_base_l {parent=thigh_l, loc=<0.4801 5.6776 -0.2486>, rot=<0 0 0 -1>}
 deform_groin_l {parent=deform_groin_base_l, loc=<0 0 0>, rot=<0 0 0 1>}
 thigh_r {parent=pelvis, loc=<-1.0855 -0.2109 8.7053>, rot=<0.0355 -0.0009 0.9993 0.0155>}
 calf_r {parent=thigh_r, loc=<42.5284 0.0019 0.7297>, rot=<0.0006 -0.0161 -0.0391 -0.9991>}
 foot_r {parent=calf_r, loc=<46.7401 2.5014 0.592>, rot=<0.0427 -0.0238 0.0353 -0.9982>}
 ball_r {parent=foot_r, loc=<3.5658 -10.8066 0.1544>, rot=<0.0001 -0.0001 -0.7063 -0.7079>}
 calf_twist_02_r {parent=calf_r, loc=<19.1546 1.7943 0.2638>, rot=<0.0237 0.0095 -0.0035 -0.9997>}
 calf_twist_01_r {parent=calf_r, loc=<35.8483 1.9197 0.5794>, rot=<0.0237 0.0095 -0.0035 -0.9997>}
 dyn_r_shoelaces_MSTR {parent=calf_r, loc=<15.16 -2.5619 1.541>, rot=<-0.0309 -0.7201 0.0276 -0.6927>}
 dyn_r_shoelace_a_01 {parent=dyn_r_shoelaces_MSTR, loc=<0.7832 -0.7109 1.0009>, rot=<0.6794 0.1959 -0.6794 0.1959>}
 dyn_r_shoelace_b_01 {parent=dyn_r_shoelaces_MSTR, loc=<-1.1702 -0.3507 1.037>, rot=<0.6493 -0.2799 -0.6493 -0.2799>}
 dyn_r_shoelace_c_01 {parent=dyn_r_shoelaces_MSTR, loc=<-0.1667 -0.0079 1.3355>, rot=<0.7071 0.0059 -0.7071 0.0059>}
 dyn_r_shoelace_c_02 {parent=dyn_r_shoelace_c_01, loc=<-6.2976 0.3419 0.3995>, rot=<0 0 0 1>}
 dyn_r_shoelace_d_01 {parent=dyn_r_shoelaces_MSTR, loc=<-0.6352 0.0045 1.3355>, rot=<0.7071 0.0059 -0.7071 0.0059>}
 dyn_r_shoelace_d_02 {parent=dyn_r_shoelace_d_01, loc=<-6.2023 0.5954 1.2368>, rot=<0 0 0 1>}
 deform_knee_in_base_r {parent=calf_r, loc=<-0.8031 5.0231 -0.2457>, rot=<0 0 0 -1>}
 deform_knee_in_r {parent=deform_knee_in_base_r, loc=<0 0 0>, rot=<0 0 0 1>}
 deform_knee_out_base_r {parent=calf_r, loc=<0 -4 0>, rot=<0 0 0 -1>}
 deform_knee_out_r {parent=deform_knee_out_base_r, loc=<0 0 0>, rot=<0 0 0 1>}
 thigh_twist_01_r {parent=thigh_r, loc=<27.952 0 0>, rot=<0.0013 0 0 -1>}
 deform_groin_base_r {parent=thigh_r, loc=<-0.4801 -5.6776 0.2486>, rot=<0 0 0 -1>}
 deform_groin_r {parent=deform_groin_base_r, loc=<0 0 0>, rot=<0 0 0 1>}
 dyn_belt_MSTR {parent=pelvis, loc=<2.0982 9.4136 -3.3382>, rot=<-0.0127 -0.707 0.0127 -0.707>}
 dyn_belt_01 {parent=dyn_belt_MSTR, loc=<0 0 0>, rot=<-0.0102 -0.1172 0.0858 -0.9893>}
 dyn_belt_02 {parent=dyn_belt_01, loc=<7.6925 0 0>, rot=<-0.0263 0.1152 0.2234 -0.9675>}
 dyn_belt_03 {parent=dyn_belt_02, loc=<7.5507 0 0>, rot=<0.1174 0.0118 0.2082 -0.9709>}
 deform_glute_base_l {parent=pelvis, loc=<-7.3466 -9.0213 -9.1245>, rot=<0.0172 0.031 -0.4477 -0.8935>}
 deform_glute_l {parent=deform_glute_base_l, loc=<0 0 0>, rot=<0 0 0 1>}
 deform_glute_base_r {parent=pelvis, loc=<-7.3464 -9.0213 9.1245>, rot=<0.031 -0.0172 0.8935 -0.4477>}
 deform_glute_r {parent=deform_glute_base_r, loc=<0 0 0>, rot=<0 0 0 1>}
]
variables [Frontend {type=bool, default=False}]
