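# Post-process AnimBP for the StrongHibiscus female player body, adding deformation and secondary motion to the pose copied from its attached parent.
# A deform Control Rig feeds weighted bone constraints, jaw/chin bone overrides, then AnimDynamics for shoelaces, belt, straps and cloth scraps.
# Additive belt and strap offsets precede a final Control Rig that receives a frontend flag.
# The event graph computes thigh/spine alphas and samples head-mesh jaw/chin transforms into variables.
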AnimBlueprint /Game/Characters/Player/Female/Medium/Bodies/F_MED_StrongHibiscus/Meshes/F_MED_StrongHibiscus_AnimBP
target_skeleton=&/F_MED_StrongHibiscus_Skeleton
graphs [
 AnimGraph {nodes [
   Output Pose(Result=Use cached pose 'Transform_Pose'()) {layerGroup=None}
   n1 = Control Rig(Source=Copy Pose From Mesh() {bUseAttachedParent=true}) {
    alphaBoolBlend {blendOption=Linear}
    controlRigAssetReference {blueprintRigClass=&/F_MED_StrongHibiscus_DeformRig.F_MED_StrongHibiscus_DeformRig_C}
   }
   n2 = "Constraint - dyn_l_cloth_scrap_fr_b_MSTR (PPP)"!(
    ComponentPose=Local To Component(LocalPose=@n1)
    Alpha=1
    ConstraintWeights_0=0.15
    ConstraintWeights_1:float=0.8
    ConstraintWeights_2:float=0.05
   ) -> (Pose) {
    alphaBoolBlend {blendOption=Linear}
    boneToModify {boneName=dyn_l_cloth_scrap_fr_b_MSTR}
    constraintSetup [{targetBone {boneName=spine_05}}, {targetBone {boneName=clavicle_l}}, {targetBone {boneName=spine_04}}]
    constraintWeights [1,1,1]
   }
   n3 = "Constraint - dyn_l_cloth_scrap_fr_c_MSTR (PPP)"!(ComponentPose=@n2, Alpha=1, ConstraintWeights_0=0.2, ConstraintWeights_1:float=0.55, ConstraintWeights_2:float=0.25) -> (Pose) {
    alphaBoolBlend {blendOption=Linear}
    boneToModify {boneName=dyn_l_cloth_scrap_fr_c_MSTR}
    constraintSetup [{targetBone {boneName=spine_05}}, {targetBone {boneName=clavicle_l}}, {targetBone {boneName=spine_04}}]
    constraintWeights [1,1,1]
   }
   n4 = "Constraint - dyn_r_cloth_scrap_fr_b_MSTR (PPP)"!(ComponentPose=@n3, Alpha=1, ConstraintWeights_0=0.15, ConstraintWeights_1:float=0.8, ConstraintWeights_2:float=0.05) -> (Pose) {
    alphaBoolBlend {blendOption=Linear}
    boneToModify {boneName=dyn_r_cloth_scrap_fr_b_MSTR}
    constraintSetup [{targetBone {boneName=spine_05}}, {targetBone {boneName=clavicle_r}}, {targetBone {boneName=spine_04}}]
    constraintWeights [1,1,1]
   }
   n5 = "Constraint - dyn_r_cloth_scrap_fr_c_MSTR (PPP)"!(ComponentPose=@n4, Alpha=1, ConstraintWeights_0=0.2, ConstraintWeights_1:float=0.55, ConstraintWeights_2:float=0.25) -> (Pose) {
    alphaBoolBlend {blendOption=Linear}
    boneToModify {boneName=dyn_r_cloth_scrap_fr_c_MSTR}
    constraintSetup [{targetBone {boneName=spine_05}}, {targetBone {boneName=clavicle_r}}, {targetBone {boneName=spine_04}}]
    constraintWeights [1,1,1]
   }
   n6 = "Constraint - dyn_knot_straps_MSTR (PPPPP)"!(
    ComponentPose=@n5
    Alpha=1
    ConstraintWeights_0=0.203
    ConstraintWeights_1:float=0.238
    ConstraintWeights_2:float=0.364
    ConstraintWeights_3:float=0.057
    ConstraintWeights_4:float=0.138
   ) -> (Pose) {
    alphaBoolBlend {blendOption=Linear}
    boneToModify {boneName=dyn_knot_straps_MSTR}
    constraintSetup [
     {targetBone {boneName=pelvis}}
     {targetBone {boneName=spine_01}}
     {targetBone {boneName=spine_02}}
     {targetBone {boneName=spine_03}}
     {targetBone {boneName=spine_04}}
    ]
    constraintWeights [1,1,1,1,1]
   }
   n7 = "Constraint - dyn_l_shoelaces_MSTR (PP)"!(ComponentPose=@n6, Alpha=1, ConstraintWeights_0=0.574, ConstraintWeights_1:float=0.426) -> (Pose) {
    alphaBoolBlend {blendOption=Linear}
    boneToModify {boneName=dyn_l_shoelaces_MSTR}
    constraintSetup [{targetBone {boneName=calf_l}}, {targetBone {boneName=calf_twist_01_l}}]
    constraintWeights [1,1]
   }
   n8 = "Constraint - dyn_r_shoelaces_MSTR (PP)"!(ComponentPose=@n7, Alpha=1, ConstraintWeights_0=0.574, ConstraintWeights_1:float=0.426) -> (Pose) {
    alphaBoolBlend {blendOption=Linear}
    boneToModify {boneName=dyn_r_shoelaces_MSTR}
    constraintSetup [{targetBone {boneName=calf_r}}, {targetBone {boneName=calf_twist_01_r}}]
    constraintWeights [1,1]
   }
   n9 = "Constraint - dyn_belt_MSTR (PP)"!(ComponentPose=@n8, Alpha=1, ConstraintWeights_0=0.85, ConstraintWeights_1:float=0.15) -> (Pose) {
    alphaBoolBlend {blendOption=Linear}
    boneToModify {boneName=dyn_belt_MSTR}
    constraintSetup [{targetBone {boneName=pelvis}}, {targetBone {boneName=deform_groin_l}}]
    constraintWeights [1,1]
   }
   n10 = "Transform (Modify) Bone - Bone: FACIAL_C_UnderChin"!(ComponentPose=@n9, bAlphaBoolEnabled:bool=True, Translation="0,0,0", Rotation="0,0,20") -> (Pose) {
    alphaInputType=Bool
    rotationMode=BMM_Replace
    rotationSpace=BCS_ParentBoneSpace
    translationMode=BMM_Replace
    translationSpace=BCS_ParentBoneSpace
    alphaBoolBlend {blendOption=Linear}
    boneToModify {boneName=FACIAL_C_UnderChin}
    rotation {roll=20}
   }
   n11 = "Transform (Modify) Bone - Bone: FACIAL_L_UnderChin"!(ComponentPose=@n10, bAlphaBoolEnabled:bool=True, Translation="0,0,0", Rotation="0,0,20") -> (Pose) {
    alphaInputType=Bool
    rotationMode=BMM_Replace
    rotationSpace=BCS_ParentBoneSpace
    translationMode=BMM_Replace
    translationSpace=BCS_ParentBoneSpace
    alphaBoolBlend {blendOption=Linear}
    boneToModify {boneName=FACIAL_L_UnderChin}
    rotation {roll=20}
   }
   n12 = "Transform (Modify) Bone - Bone: FACIAL_R_UnderChin"!(ComponentPose=@n11, bAlphaBoolEnabled:bool=True, Translation="0,0,0", Rotation="0,0,20") -> (Pose) {
    alphaInputType=Bool
    rotationMode=BMM_Replace
    rotationSpace=BCS_ParentBoneSpace
    translationMode=BMM_Replace
    translationSpace=BCS_ParentBoneSpace
    alphaBoolBlend {blendOption=Linear}
    boneToModify {boneName=FACIAL_R_UnderChin}
    rotation {roll=20}
   }
   n13 = "Transform (Modify) Bone - Bone: FACIAL_L_Jawline"!(ComponentPose=@n12, bAlphaBoolEnabled:bool=True, Translation="0,0,0", Rotation="0,0,20") -> (Pose) {
    alphaInputType=Bool
    rotationMode=BMM_Replace
    rotationSpace=BCS_ParentBoneSpace
    translationMode=BMM_Replace
    translationSpace=BCS_ParentBoneSpace
    alphaBoolBlend {blendOption=Linear}
    boneToModify {boneName=FACIAL_L_Jawline}
    rotation {roll=20}
   }
   n14 = "Transform (Modify) Bone - Bone: FACIAL_R_Jawline"!(ComponentPose=@n13, bAlphaBoolEnabled:bool=True, Translation="0,0,0", Rotation="0,0,20") -> (Pose) {
    alphaInputType=Bool
    rotationMode=BMM_Replace
    rotationSpace=BCS_ParentBoneSpace
    translationMode=BMM_Replace
    translationSpace=BCS_ParentBoneSpace
    alphaBoolBlend {blendOption=Linear}
    boneToModify {boneName=FACIAL_R_Jawline}
    rotation {roll=20}
   }
   n15 = "Transform (Modify) Bone - Bone: FACIAL_C_Jaw"!(ComponentPose=@n14, bAlphaBoolEnabled:bool=True, Translation="0,0,0", Rotation="0,0,20") -> (Pose) {
    alphaInputType=Bool
    rotationMode=BMM_Replace
    rotationSpace=BCS_ParentBoneSpace
    translationMode=BMM_Replace
    translationSpace=BCS_ParentBoneSpace
    alphaBoolBlend {blendOption=Linear}
    boneToModify {boneName=FACIAL_C_Jaw}
    rotation {roll=20}
   }
   n16 = "Transform (Modify) Bone - Bone: FACIAL_C_Jawline"!(ComponentPose=@n15, bAlphaBoolEnabled:bool=True, Translation="0,0,0", Rotation="0,0,20") -> (Pose) {
    alphaInputType=Bool
    rotationMode=BMM_Replace
    rotationSpace=BCS_ParentBoneSpace
    translationMode=BMM_Replace
    translationSpace=BCS_ParentBoneSpace
    alphaBoolBlend {blendOption=Linear}
    boneToModify {boneName=FACIAL_C_Jawline}
    rotation {roll=20}
   }
   Save cached pose 'Start_Pose'(Pose=Component To Local(ComponentPose=@n16)) {cacheName=Start_Pose}
   n17 = Anim Dynamics - Bone: dyn_l_shoelace_a_01(ComponentPose=Local To Component(LocalPose=Use cached pose 'Start_Pose'()), Alpha=1, ExternalForce="0,0,0") {
    angularDampingOverride=0.99
    angularSpringConstant=500
    bAngularSpring=true
    bGravityOverrideInSimSpace=true
    bOverrideAngularDamping=true
    bUseGravityOverride=true
    simulationSpace=RootRelative
    alphaBoolBlend {blendOption=Linear}
    boundBone {boneName=dyn_l_shoelace_a_01}
    componentAppliedLinearAccClamp=<3000 3000 3000>
    componentLinearAccScale=<0.3 0.3 0.3>
    componentLinearVelScale=<0.3 0.3 0.3>
    gravityOverride {z=-950}
    physicsBodyDefinitions [{
      boundBone {boneName=dyn_l_shoelace_a_01}
      boxExtents=<15 15 15>
      constraintSetup {angularLimitsMax=<5 20 5>, angularLimitsMin=<-5 -30 -80>, angularTarget=<0 0 0>}
      localJointOffset=<2 0.1 0>
     }]
   }
   n18 = Anim Dynamics - Bone: dyn_l_shoelace_b_01(ComponentPose=@n17, Alpha=1, ExternalForce="0,0,0") {
    angularDampingOverride=0.99
    angularSpringConstant=500
    bAngularSpring=true
    bGravityOverrideInSimSpace=true
    bOverrideAngularDamping=true
    bUseGravityOverride=true
    simulationSpace=RootRelative
    alphaBoolBlend {blendOption=Linear}
    boundBone {boneName=dyn_l_shoelace_b_01}
    componentAppliedLinearAccClamp=<3000 3000 3000>
    componentLinearAccScale=<0.3 0.3 0.3>
    componentLinearVelScale=<0.3 0.3 0.3>
    gravityOverride {z=-950}
    physicsBodyDefinitions [{
      boundBone {boneName=dyn_l_shoelace_b_01}
      boxExtents=<15 15 15>
      constraintSetup {angularLimitsMax=<5 30 5>, angularLimitsMin=<-5 -20 -80>, angularTarget=<0 0 0>}
      localJointOffset=<2 0.1 0>
     }]
   }
   n19 = Anim Dynamics - Bone: dyn_l_shoelace_c_01(ComponentPose=@n18, Alpha=1, ExternalForce="0,0,0") {
    angularDampingOverride=0.99
    angularSpringConstant=500
    bAngularSpring=true
    bGravityOverrideInSimSpace=true
    bOverrideAngularDamping=true
    bUseGravityOverride=true
    simulationSpace=RootRelative
    alphaBoolBlend {blendOption=Linear}
    boundBone {boneName=dyn_l_shoelace_c_01}
    componentAppliedLinearAccClamp=<3000 3000 3000>
    componentLinearAccScale=<0.3 0.3 0.3>
    componentLinearVelScale=<0.3 0.3 0.3>
    gravityOverride {z=-950}
    physicsBodyDefinitions [{
      boundBone {boneName=dyn_l_shoelace_c_01}
      boxExtents=<15 15 15>
      constraintSetup {angularLimitsMax=<5 10 5>, angularLimitsMin=<-5 -10 -80>, angularTarget=<0 0 0>}
      localJointOffset=<2 0.1 0>
     }]
   }
   n20 = Anim Dynamics - Bone: dyn_l_shoelace_c_02(ComponentPose=@n19, Alpha=1, ExternalForce="0,0,0") {
    angularDampingOverride=0.99
    angularSpringConstant=500
    bAngularSpring=true
    bGravityOverrideInSimSpace=true
    bOverrideAngularDamping=true
    bUseGravityOverride=true
    simulationSpace=RootRelative
    alphaBoolBlend {blendOption=Linear}
    boundBone {boneName=dyn_l_shoelace_c_02}
    componentAppliedLinearAccClamp=<3000 3000 3000>
    componentLinearAccScale=<0.3 0.3 0.3>
    componentLinearVelScale=<0.3 0.3 0.3>
    gravityOverride {z=-950}
    physicsBodyDefinitions [{
      boundBone {boneName=dyn_l_shoelace_c_02}
      boxExtents=<10 10 10>
      constraintSetup {angularLimitsMax=<5 30 5>, angularLimitsMin=<-5 -30 -30>, angularTarget=<0 0 0>}
      localJointOffset=<2 0 0>
     }]
   }
   n21 = Anim Dynamics - Bone: dyn_l_shoelace_d_01(ComponentPose=@n20, Alpha=1, ExternalForce="0,0,0") {
    angularDampingOverride=0.99
    angularSpringConstant=500
    bAngularSpring=true
    bGravityOverrideInSimSpace=true
    bOverrideAngularDamping=true
    bUseGravityOverride=true
    simulationSpace=RootRelative
    alphaBoolBlend {blendOption=Linear}
    boundBone {boneName=dyn_l_shoelace_d_01}
    componentAppliedLinearAccClamp=<3000 3000 3000>
    componentLinearAccScale=<0.3 0.3 0.3>
    componentLinearVelScale=<0.3 0.3 0.3>
    gravityOverride {z=-950}
    physicsBodyDefinitions [{
      boundBone {boneName=dyn_l_shoelace_d_01}
      boxExtents=<15 15 15>
      constraintSetup {angularLimitsMax=<5 180 5>, angularLimitsMin=<-5 -20 -80>, angularTarget=<0 0 0>}
      localJointOffset=<2 0.1 0>
     }]
   }
   n22 = Anim Dynamics - Bone: dyn_l_shoelace_d_02(ComponentPose=@n21, Alpha=1, ExternalForce="0,0,0") {
    angularDampingOverride=0.99
    angularSpringConstant=500
    bAngularSpring=true
    bGravityOverrideInSimSpace=true
    bOverrideAngularDamping=true
    bUseGravityOverride=true
    simulationSpace=RootRelative
    alphaBoolBlend {blendOption=Linear}
    boundBone {boneName=dyn_l_shoelace_d_02}
    componentAppliedLinearAccClamp=<3000 3000 3000>
    componentLinearAccScale=<0.3 0.3 0.3>
    componentLinearVelScale=<0.3 0.3 0.3>
    gravityOverride {z=-950}
    physicsBodyDefinitions [{
      boundBone {boneName=dyn_l_shoelace_d_02}
      boxExtents=<10 10 10>
      constraintSetup {angularLimitsMax=<5 30 5>, angularLimitsMin=<-5 -30 -30>, angularTarget=<0 0 0>}
      localJointOffset=<2 0 0>
     }]
   }
   n23 = Anim Dynamics - Bone: dyn_r_shoelace_a_01(ComponentPose=@n22, Alpha=1, ExternalForce="0,0,0") {
    angularDampingOverride=0.99
    angularSpringConstant=500
    bAngularSpring=true
    bGravityOverrideInSimSpace=true
    bOverrideAngularDamping=true
    bUseGravityOverride=true
    simulationSpace=RootRelative
    alphaBoolBlend {blendOption=Linear}
    boundBone {boneName=dyn_r_shoelace_a_01}
    componentAppliedLinearAccClamp=<3000 3000 3000>
    componentLinearAccScale=<0.3 0.3 0.3>
    componentLinearVelScale=<0.3 0.3 0.3>
    gravityOverride {z=-950}
    physicsBodyDefinitions [{
      boundBone {boneName=dyn_r_shoelace_a_01}
      boxExtents=<15 15 15>
      constraintSetup {angularLimitsMax=<5 10 5>, angularLimitsMin=<-5 -30 -80>, angularTarget=<0 0 0>}
      localJointOffset=<-2 0.1 0>
     }]
   }
   n24 = Anim Dynamics - Bone: dyn_r_shoelace_b_01(ComponentPose=@n23, Alpha=1, ExternalForce="0,0,0") {
    angularDampingOverride=0.99
    angularSpringConstant=500
    bAngularSpring=true
    bGravityOverrideInSimSpace=true
    bOverrideAngularDamping=true
    bUseGravityOverride=true
    simulationSpace=RootRelative
    alphaBoolBlend {blendOption=Linear}
    boundBone {boneName=dyn_r_shoelace_b_01}
    componentAppliedLinearAccClamp=<3000 3000 3000>
    componentLinearAccScale=<0.3 0.3 0.3>
    componentLinearVelScale=<0.3 0.3 0.3>
    gravityOverride {z=-950}
    physicsBodyDefinitions [{
      boundBone {boneName=dyn_r_shoelace_b_01}
      boxExtents=<15 15 15>
      constraintSetup {angularLimitsMax=<5 20 5>, angularLimitsMin=<-5 -50 -80>, angularTarget=<0 0 0>}
      localJointOffset=<-2 0.1 0>
     }]
   }
   n25 = Anim Dynamics - Bone: dyn_r_shoelace_c_01(ComponentPose=@n24, Alpha=1, ExternalForce="0,0,0") {
    angularDampingOverride=0.99
    angularSpringConstant=500
    bAngularSpring=true
    bGravityOverrideInSimSpace=true
    bOverrideAngularDamping=true
    bUseGravityOverride=true
    simulationSpace=RootRelative
    alphaBoolBlend {blendOption=Linear}
    boundBone {boneName=dyn_r_shoelace_c_01}
    componentAppliedLinearAccClamp=<3000 3000 3000>
    componentLinearAccScale=<0.3 0.3 0.3>
    componentLinearVelScale=<0.3 0.3 0.3>
    gravityOverride {z=-950}
    physicsBodyDefinitions [{
      boundBone {boneName=dyn_r_shoelace_c_01}
      boxExtents=<15 15 15>
      constraintSetup {angularLimitsMax=<5 10 5>, angularLimitsMin=<-5 -10 -80>, angularTarget=<0 0 0>}
      localJointOffset=<-2 0.1 0>
     }]
   }
   n26 = Anim Dynamics - Bone: dyn_r_shoelace_c_02(ComponentPose=@n25, Alpha=1, ExternalForce="0,0,0") {
    angularDampingOverride=0.99
    angularSpringConstant=500
    bAngularSpring=true
    bGravityOverrideInSimSpace=true
    bOverrideAngularDamping=true
    bUseGravityOverride=true
    simulationSpace=RootRelative
    alphaBoolBlend {blendOption=Linear}
    boundBone {boneName=dyn_r_shoelace_c_02}
    componentAppliedLinearAccClamp=<3000 3000 3000>
    componentLinearAccScale=<0.3 0.3 0.3>
    componentLinearVelScale=<0.3 0.3 0.3>
    gravityOverride {z=-950}
    physicsBodyDefinitions [{
      boundBone {boneName=dyn_r_shoelace_c_02}
      boxExtents=<10 10 10>
      constraintSetup {angularLimitsMax=<5 30 5>, angularLimitsMin=<-5 -30 -20>, angularTarget=<0 0 0>}
      localJointOffset=<-2 0 0>
     }]
   }
   n27 = Anim Dynamics - Bone: dyn_r_shoelace_d_01(ComponentPose=@n26, Alpha=1, ExternalForce="0,0,0") {
    angularDampingOverride=0.99
    angularSpringConstant=500
    bAngularSpring=true
    bGravityOverrideInSimSpace=true
    bOverrideAngularDamping=true
    bUseGravityOverride=true
    simulationSpace=RootRelative
    alphaBoolBlend {blendOption=Linear}
    boundBone {boneName=dyn_r_shoelace_d_01}
    componentAppliedLinearAccClamp=<3000 3000 3000>
    componentLinearAccScale=<0.3 0.3 0.3>
    componentLinearVelScale=<0.3 0.3 0.3>
    gravityOverride {z=-950}
    physicsBodyDefinitions [{
      boundBone {boneName=dyn_r_shoelace_d_01}
      boxExtents=<15 15 15>
      constraintSetup {angularLimitsMax=<5 180 5>, angularLimitsMin=<-5 -20 -80>, angularTarget=<0 0 0>}
      localJointOffset=<-2 0.1 0>
     }]
   }
   n28 = Anim Dynamics - Bone: dyn_r_shoelace_d_02(ComponentPose=@n27, Alpha=1, ExternalForce="0,0,0") {
    angularDampingOverride=0.99
    angularSpringConstant=500
    bAngularSpring=true
    bGravityOverrideInSimSpace=true
    bOverrideAngularDamping=true
    bUseGravityOverride=true
    simulationSpace=RootRelative
    alphaBoolBlend {blendOption=Linear}
    boundBone {boneName=dyn_r_shoelace_d_02}
    componentAppliedLinearAccClamp=<3000 3000 3000>
    componentLinearAccScale=<0.3 0.3 0.3>
    componentLinearVelScale=<0.3 0.3 0.3>
    gravityOverride {z=-950}
    physicsBodyDefinitions [{
      boundBone {boneName=dyn_r_shoelace_d_02}
      boxExtents=<10 10 10>
      constraintSetup {angularLimitsMax=<5 30 5>, angularLimitsMin=<-5 -30 -20>, angularTarget=<0 0 0>}
      localJointOffset=<-2 0 0>
     }]
   }
   n29 = Anim Dynamics - Bone: dyn_belt_01(ComponentPose=@n28, Alpha=1, ExternalForce="0,0,0") {
    angularDampingOverride=0.99
    angularSpringConstant=500
    bAngularSpring=true
    bGravityOverrideInSimSpace=true
    bOverrideAngularDamping=true
    bUseGravityOverride=true
    simulationSpace=BoneRelative
    alphaBoolBlend {blendOption=Linear}
    boundBone {boneName=dyn_belt_01}
    componentAppliedLinearAccClamp=<3000 3000 3000>
    componentLinearAccScale=<0.3 0.3 0.3>
    componentLinearVelScale=<0.3 0.3 0.3>
    gravityOverride {z=-980}
    physicsBodyDefinitions [{
      boundBone {boneName=dyn_belt_01}
      boxExtents=<15 15 15>
      constraintSetup {angularLimitsMax=<5 0 10>, angularLimitsMin=<-5 -10 -10>, angularTarget=<0 0 0>}
      localJointOffset=<1 0 0>
     }]
    relativeSpaceBone {boneName=pelvis}
   }
   n30 = Anim Dynamics - Bone: dyn_belt_02(ComponentPose=@n29, Alpha=1, ExternalForce="0,0,0") {
    angularDampingOverride=0.99
    angularSpringConstant=500
    bAngularSpring=true
    bGravityOverrideInSimSpace=true
    bOverrideAngularDamping=true
    bUseGravityOverride=true
    simulationSpace=RootRelative
    alphaBoolBlend {blendOption=Linear}
    boundBone {boneName=dyn_belt_02}
    componentAppliedLinearAccClamp=<3000 3000 3000>
    componentLinearAccScale=<0.3 0.3 0.3>
    componentLinearVelScale=<0.3 0.3 0.3>
    gravityOverride {z=-700}
    physicsBodyDefinitions [{
      boundBone {boneName=dyn_belt_02}
      boxExtents=<15 15 15>
      constraintSetup {angularLimitsMax=<10 0 20>, angularLimitsMin=<-10 -20 -10>, angularTarget=<0 0 0>}
      localJointOffset=<2 0 0>
     }]
    relativeSpaceBone {boneName=dyn_belt_01}
   }
   n31 = Anim Dynamics - Bone: dyn_belt_03(ComponentPose=@n30, Alpha=1, ExternalForce="0,0,0") {
    angularDampingOverride=0.9
    angularSpringConstant=300
    bAngularSpring=true
    bGravityOverrideInSimSpace=true
    bOverrideAngularDamping=true
    bUseGravityOverride=true
    simulationSpace=RootRelative
    alphaBoolBlend {blendOption=Linear}
    boundBone {boneName=dyn_belt_03}
    componentAppliedLinearAccClamp=<3000 3000 3000>
    componentLinearAccScale=<0.3 0.3 0.3>
    componentLinearVelScale=<0.3 0.3 0.3>
    gravityOverride {z=-800}
    physicsBodyDefinitions [{
      boundBone {boneName=dyn_belt_03}
      boxExtents=<10 10 10>
      constraintSetup {angularLimitsMax=<5 0 20>, angularLimitsMin=<-5 -10 -15>, angularTarget=<0 0 0>}
      localJointOffset=<2 0 0>
     }]
    relativeSpaceBone {boneName=dyn_belt_02}
   }
   n32 = Anim Dynamics - Bone: dyn_knot_strap_a_01(ComponentPose=@n31, Alpha=1, ExternalForce="0,0,0") {
    angularDampingOverride=0.99
    angularSpringConstant=500
    bAngularSpring=true
    bGravityOverrideInSimSpace=true
    bOverrideAngularDamping=true
    bUseGravityOverride=true
    alphaBoolBlend {blendOption=Linear}
    boundBone {boneName=dyn_knot_strap_a_01}
    componentAppliedLinearAccClamp=<3000 3000 3000>
    componentLinearAccScale=<0.3 0.3 0.3>
    componentLinearVelScale=<0.3 0.3 0.3>
    gravityOverride {z=-950}
    physicsBodyDefinitions [{
      boundBone {boneName=dyn_knot_strap_a_01}
      boxExtents=<15 15 15>
      constraintSetup {angularLimitsMax=<10 20 5>, angularLimitsMin=<-10 -5 -30>, angularTarget=<0 0 0>}
      localJointOffset=<3 0.8 0>
     }]
    relativeSpaceBone {boneName=spine_03}
   }
   n33 = Anim Dynamics - Bone: dyn_knot_strap_a_02(ComponentPose=@n32, Alpha=1, ExternalForce="0,0,0") {
    angularDampingOverride=0.99
    angularSpringConstant=500
    bAngularSpring=true
    bGravityOverrideInSimSpace=true
    bOverrideAngularDamping=true
    bUseGravityOverride=true
    simulationSpace=RootRelative
    alphaBoolBlend {blendOption=Linear}
    boundBone {boneName=dyn_knot_strap_a_02}
    componentAppliedLinearAccClamp=<3000 3000 3000>
    componentLinearAccScale=<0.3 0.3 0.3>
    componentLinearVelScale=<0.3 0.3 0.3>
    gravityOverride {z=-900}
    physicsBodyDefinitions [{
      boundBone {boneName=dyn_knot_strap_a_02}
      boxExtents=<15 15 15>
      constraintSetup {angularLimitsMax=<10 30 0>, angularLimitsMin=<-10 -10 -40>, angularTarget=<0 0 0>}
      localJointOffset=<3 0.8 0>
     }]
    relativeSpaceBone {boneName=spine_03}
   }
   n34 = Anim Dynamics - Bone: dyn_knot_strap_b_01(ComponentPose=@n33, Alpha=1, ExternalForce="0,0,0") {
    angularDampingOverride=0.99
    angularSpringConstant=500
    bAngularSpring=true
    bGravityOverrideInSimSpace=true
    bOverrideAngularDamping=true
    bUseGravityOverride=true
    alphaBoolBlend {blendOption=Linear}
    boundBone {boneName=dyn_knot_strap_b_01}
    componentAppliedLinearAccClamp=<3000 3000 3000>
    componentLinearAccScale=<0.3 0.3 0.3>
    componentLinearVelScale=<0.3 0.3 0.3>
    gravityOverride {z=-950}
    physicsBodyDefinitions [{
      boundBone {boneName=dyn_knot_strap_b_01}
      boxExtents=<15 15 15>
      constraintSetup {angularLimitsMax=<10 20 5>, angularLimitsMin=<-10 -5 -20>, angularTarget=<0 0 0>}
      localJointOffset=<4 0.8 0>
     }]
    relativeSpaceBone {boneName=spine_03}
   }
   n35 = Anim Dynamics - Bone: dyn_knot_strap_b_02(ComponentPose=@n34, Alpha=1, ExternalForce="0,0,0") {
    angularDampingOverride=0.99
    angularSpringConstant=500
    bAngularSpring=true
    bGravityOverrideInSimSpace=true
    bOverrideAngularDamping=true
    bUseGravityOverride=true
    simulationSpace=RootRelative
    alphaBoolBlend {blendOption=Linear}
    boundBone {boneName=dyn_knot_strap_b_02}
    componentAppliedLinearAccClamp=<3000 3000 3000>
    componentLinearAccScale=<0.3 0.3 0.3>
    componentLinearVelScale=<0.3 0.3 0.3>
    gravityOverride {z=-900}
    physicsBodyDefinitions [{
      boundBone {boneName=dyn_knot_strap_b_02}
      boxExtents=<15 15 15>
      constraintSetup {angularLimitsMax=<10 30 5>, angularLimitsMin=<-10 -10 -30>, angularTarget=<0 0 0>}
      localJointOffset=<4 0.8 0>
     }]
    relativeSpaceBone {boneName=spine_03}
   }
   n36 = Anim Dynamics - Bone: dyn_l_cloth_scrap_mid_01(ComponentPose=@n35, Alpha=1, ExternalForce="0,0,0") {
    angularDampingOverride=0.99
    angularSpringConstant=3500
    bAngularSpring=true
    bOverrideAngularDamping=true
    bUseGravityOverride=true
    bUsePlanarLimit=false
    gravityScale=0.35
    simulationSpace=RootRelative
    alphaBoolBlend {blendOption=Linear}
    boundBone {boneName=dyn_l_cloth_scrap_mid_01}
    componentAppliedLinearAccClamp=<1000 1000 1000>
    componentLinearAccScale=<0.6 0.5 0.5>
    componentLinearVelScale=<0.6 0.5 0.5>
    gravityOverride {z=-980}
    physicsBodyDefinitions [{
      boundBone {boneName=dyn_l_cloth_scrap_mid_01}
      boxExtents=<10 10 10>
      constraintSetup {angularLimitsMax=<5 5 0>, angularLimitsMin=<-5 -5 -30>, angularTarget=<1 0 0>}
      localJointOffset=<1 0 0>
     }]
   }
   n37 = Anim Dynamics - Bone: dyn_l_cloth_scrap_fr_a_01(ComponentPose=@n36, Alpha=1, ExternalForce="0,0,0") {
    angularDampingOverride=0.99
    angularSpringConstant=3500
    bAngularSpring=true
    bOverrideAngularDamping=true
    bUseGravityOverride=true
    bUsePlanarLimit=false
    gravityScale=0.35
    simulationSpace=RootRelative
    alphaBoolBlend {blendOption=Linear}
    boundBone {boneName=dyn_l_cloth_scrap_fr_a_01}
    componentAppliedLinearAccClamp=<1000 1000 1000>
    componentLinearAccScale=<0.6 0.5 0.5>
    componentLinearVelScale=<0.6 0.5 0.5>
    gravityOverride {z=-980}
    physicsBodyDefinitions [{
      boundBone {boneName=dyn_l_cloth_scrap_fr_a_01}
      boxExtents=<10 10 10>
      constraintSetup {angularLimitsMax=<5 5 0>, angularLimitsMin=<-5 -5 -30>, angularTarget=<1 0 0>}
      localJointOffset=<1 0 0>
     }]
   }
   n38 = Anim Dynamics - Bone: dyn_l_cloth_scrap_bk_a_01(ComponentPose=@n37, Alpha=1, ExternalForce="0,0,0") {
    angularDampingOverride=0.99
    angularSpringConstant=3500
    bAngularSpring=true
    bOverrideAngularDamping=true
    bUseGravityOverride=true
    bUsePlanarLimit=false
    gravityScale=0.35
    simulationSpace=RootRelative
    alphaBoolBlend {blendOption=Linear}
    boundBone {boneName=dyn_l_cloth_scrap_bk_a_01}
    componentAppliedLinearAccClamp=<1000 1000 1000>
    componentLinearAccScale=<0.6 0.5 0.5>
    componentLinearVelScale=<0.6 0.5 0.5>
    gravityOverride {z=-980}
    physicsBodyDefinitions [{
      boundBone {boneName=dyn_l_cloth_scrap_bk_a_01}
      boxExtents=<10 10 10>
      constraintSetup {angularLimitsMax=<5 5 0>, angularLimitsMin=<-5 -5 -30>, angularTarget=<1 0 0>}
      localJointOffset=<1 0 0>
     }]
   }
   n39 = Anim Dynamics - Bone: dyn_r_cloth_scrap_mid_01(ComponentPose=@n38, Alpha=1, ExternalForce="0,0,0") {
    angularDampingOverride=0.99
    angularSpringConstant=3500
    bAngularSpring=true
    bOverrideAngularDamping=true
    bUseGravityOverride=true
    bUsePlanarLimit=false
    gravityScale=0.35
    simulationSpace=RootRelative
    alphaBoolBlend {blendOption=Linear}
    boundBone {boneName=dyn_r_cloth_scrap_mid_01}
    componentAppliedLinearAccClamp=<1000 1000 1000>
    componentLinearAccScale=<0.6 0.5 0.5>
    componentLinearVelScale=<0.6 0.5 0.5>
    gravityOverride {z=-980}
    physicsBodyDefinitions [{
      boundBone {boneName=dyn_r_cloth_scrap_mid_01}
      boxExtents=<10 10 10>
      constraintSetup {angularLimitsMax=<5 5 0>, angularLimitsMin=<-5 -5 -30>, angularTarget=<1 0 0>}
      localJointOffset=<1 0 0>
     }]
   }
   n40 = Anim Dynamics - Bone: dyn_r_cloth_scrap_fr_a_01(ComponentPose=@n39, Alpha=1, ExternalForce="0,0,0") {
    angularDampingOverride=0.99
    angularSpringConstant=3500
    bAngularSpring=true
    bOverrideAngularDamping=true
    bUseGravityOverride=true
    bUsePlanarLimit=false
    gravityScale=0.35
    simulationSpace=RootRelative
    alphaBoolBlend {blendOption=Linear}
    boundBone {boneName=dyn_r_cloth_scrap_fr_a_01}
    componentAppliedLinearAccClamp=<1000 1000 1000>
    componentLinearAccScale=<0.6 0.5 0.5>
    componentLinearVelScale=<0.6 0.5 0.5>
    gravityOverride {z=-980}
    physicsBodyDefinitions [{
      boundBone {boneName=dyn_r_cloth_scrap_fr_a_01}
      boxExtents=<10 10 10>
      constraintSetup {angularLimitsMax=<5 5 0>, angularLimitsMin=<-5 -5 -30>, angularTarget=<1 0 0>}
      localJointOffset=<1 0 0>
     }]
   }
   n41 = Anim Dynamics - Bone: dyn_r_cloth_scrap_bk_a_01(ComponentPose=@n40, Alpha=1, ExternalForce="0,0,0") {
    angularDampingOverride=0.99
    angularSpringConstant=3500
    bAngularSpring=true
    bOverrideAngularDamping=true
    bUseGravityOverride=true
    bUsePlanarLimit=false
    gravityScale=0.35
    simulationSpace=RootRelative
    alphaBoolBlend {blendOption=Linear}
    boundBone {boneName=dyn_r_cloth_scrap_bk_a_01}
    componentAppliedLinearAccClamp=<1000 1000 1000>
    componentLinearAccScale=<0.6 0.5 0.5>
    componentLinearVelScale=<0.6 0.5 0.5>
    gravityOverride {z=-980}
    physicsBodyDefinitions [{
      boundBone {boneName=dyn_r_cloth_scrap_bk_a_01}
      boxExtents=<10 10 10>
      constraintSetup {angularLimitsMax=<5 5 0>, angularLimitsMin=<-5 -5 -30>, angularTarget=<1 0 0>}
      localJointOffset=<1 0 0>
     }]
   }
   Save cached pose 'Dynamic_Pose'(Pose=Component To Local(ComponentPose=@n41)) {cacheName=Dynamic_Pose}
   n42 = "Transform (Modify) Bone - Bone: dyn_belt_MSTR"!(ComponentPose=Local To Component(LocalPose=Use cached pose 'Dynamic_Pose'()), Alpha=1, Translation="0,0,0", Rotation="55.181,2.961,-7.194") -> (Pose) {rotationMode=BMM_Additive, alphaBoolBlend {blendOption=Linear}, boneToModify {boneName=dyn_belt_MSTR}, rotation=<rot 55.181 2.961 -7.194>}
   n43 = "Transform (Modify) Bone - Bone: dyn_belt_02"!(ComponentPose=@n42, Alpha=1, Translation="-0.000,-0.000,-2.755", Rotation="-38.639,0.000,-0.000") -> (Pose) {
    rotationMode=BMM_Additive
    translationMode=BMM_Additive
    alphaBoolBlend {blendOption=Linear}
    boneToModify {boneName=dyn_belt_02}
    rotation {pitch=-38.639}
    translation {z=-2.755}
   }
   n44 = "Transform (Modify) Bone - Bone: dyn_knot_straps_MSTR"!(ComponentPose=@n43, Alpha=1, Translation="0,0,0", Rotation="0,30,-30") -> (Pose) {
    rotationMode=BMM_Additive
    translationMode=BMM_Additive
    alphaBoolBlend {blendOption=Linear}
    boneToModify {boneName=dyn_knot_straps_MSTR}
    rotation {roll=-30, yaw=30}
   }
   n45 = Control Rig!(Source=Component To Local(ComponentPose=@n44), Alpha, Frontend:bool=Get bIsInFrontEnd()) -> (Pose) {
    alphaBoolBlend {blendOption=Linear}
    controlRigAssetReference {blueprintRigClass=&/F_MED_StrongHibiscus_CtrlRig.F_MED_StrongHibiscus_CtrlRig_C}
   }
   Save cached pose 'Transform_Pose'(Pose=@n45) {cacheName=Transform_Pose}
  ]}
 EventGraph {nodes [
   Event BlueprintUpdateAnimation()
   > n1 = Sequence()
   n2 = Get SkelMesh()
   > Set Alpha_Thigh_L_Front(Alpha_Thigh_L_Front:double=GetAlphaFromReferenceTransformDelta(SkeletalMeshComponent=@n2, BoneName=thigh_l, BaseName=pelvis, InputRange=(X=0,Y=-110), TwistAxis="1,0,0")) -> (Output_Get:double)
   > Set Alpha_Spine_Front(Alpha_Spine_Front:double=GetAlphaFromReferenceTransformDelta(SkeletalMeshComponent=@n2, BoneName=spine_02, BaseName=pelvis, InputRange=(X=5,Y=30), TwistAxis="1,0,0")) -> (Output_Get:double)
   n3 = Get HeadMesh() -> (HeadMesh:Skeletal Mesh Component Object Reference)
   Is Valid(exec=@n1.then_1, InputObject=@n3)
   n4 = GetSocketBoneName!(self:Skinned Mesh Component Object Reference=@n3, InSocketName:Name=FACIAL_C_FacialRoot) -> (ReturnValue:Name=None)
   > Set IsFaceBones(IsFaceBones:bool="Not Equal Exactly (String)"(A="To String (Name)"(InName=@n4), B=None)) -> (Output_Get:bool)
   n5 = Get HeadMesh()
   n6 = BreakTransform(InTransform=GetSocketTransform(self=@n5, InSocketName=FACIAL_C_UnderChin, TransformSpace=RTS_ParentBoneSpace))
   > Set FACIAL_C_UnderChin_Translation(FACIAL_C_UnderChin_Translation:Vector=@n6) -> (Output_Get:Vector)
   > Set FACIAL_C_UnderChin_Rotation(FACIAL_C_UnderChin_Rotation:Rotator=@n6.Rotation) -> (Output_Get:Rotator)
   n7 = BreakTransform(InTransform=GetSocketTransform(self=@n5, InSocketName=FACIAL_L_UnderChin, TransformSpace=RTS_ParentBoneSpace))
   > Set FACIAL_L_UnderChin_Translation(FACIAL_L_UnderChin_Translation:Vector=@n7) -> (Output_Get:Vector)
   > Set FACIAL_L_UnderChin_Rotation(FACIAL_L_UnderChin_Rotation:Rotator=@n7.Rotation) -> (Output_Get:Rotator)
   n8 = BreakTransform(InTransform=GetSocketTransform(self=@n5, InSocketName=FACIAL_R_UnderChin, TransformSpace=RTS_ParentBoneSpace))
   > Set FACIAL_R_UnderChin_Translation(FACIAL_R_UnderChin_Translation:Vector=@n8) -> (Output_Get:Vector)
   > Set FACIAL_R_UnderChin_Rotation(FACIAL_R_UnderChin_Rotation:Rotator=@n8.Rotation) -> (Output_Get:Rotator)
   n9 = BreakTransform(InTransform=GetSocketTransform(self=Get HeadMesh(), InSocketName=FACIAL_C_Jaw, TransformSpace=RTS_ParentBoneSpace))
   > Set FACIAL_C_Jaw_Tranform(FACIAL_C_Jaw_Tranform:Vector=@n9) -> (Output_Get:Vector)
   > Set FACIAL_C_Jaw_Rotation(FACIAL_C_Jaw_Rotation:Rotator=@n9.Rotation) -> (Output_Get:Rotator)
   n10 = Get HeadMesh()
   n11 = BreakTransform(InTransform=GetSocketTransform(self=@n10, InSocketName=FACIAL_C_Jawline, TransformSpace=RTS_ParentBoneSpace))
   > Set FACIAL_C_Jawline_Translation(FACIAL_C_Jawline_Translation:Vector=@n11) -> (Output_Get:Vector)
   > Set FACIAL_C_Jawline_Rotation(FACIAL_C_Jawline_Rotation:Rotator=@n11.Rotation) -> (Output_Get:Rotator)
   n12 = BreakTransform(InTransform=GetSocketTransform(self=@n10, InSocketName=FACIAL_L_Jawline, TransformSpace=RTS_ParentBoneSpace))
   > Set FACIAL_L_Jawline_Translation(FACIAL_L_Jawline_Translation:Vector=@n12) -> (Output_Get:Vector)
   > Set FACIAL_L_Jawline_Rotation(FACIAL_L_Jawline_Rotation:Rotator=@n12.Rotation) -> (Output_Get:Rotator)
   n13 = BreakTransform(InTransform=GetSocketTransform(self=@n10, InSocketName=FACIAL_R_Jawline, TransformSpace=RTS_ParentBoneSpace))
   > Set FACIAL_R_Jawline_Translation(FACIAL_R_Jawline_Translation:Vector=@n13) -> (Output_Get:Vector)
   > Set FACIAL_R_Jawline_Rotation(FACIAL_R_Jawline_Rotation:Rotator=@n13.Rotation) -> (Output_Get:Rotator)
   n14 = Event OnCharacterCustomizationLoaded()
   > Set SkelMesh(SkelMesh=GetPartSkeletalMeshComponent())
   > Set HeadMesh(HeadMesh:Skeletal Mesh Component Object Reference=Get Array Item with EFort Custom Part Type(Array=@n14.CustomMeshParts, PartType=Head)) -> (Output_Get:Skeletal Mesh Component Object Reference)
   Event BlueprintInitializeAnimation()
   > Set SkelMesh(SkelMesh=GetOwnerSkeletalMeshComponent())
  ]}
]
variables [
 Alpha_Spine_Front {type=number}
 Alpha_Thigh_L_Front {type=number}
 FACIAL_C_Jaw_Rotation {type=Rotator}
 FACIAL_C_Jaw_Tranform {type=Vector}
 FACIAL_C_Jawline_Rotation {type=Rotator}
 FACIAL_C_Jawline_Translation {type=Vector}
 FACIAL_C_UnderChin_Rotation {type=Rotator}
 FACIAL_C_UnderChin_Translation {type=Vector}
 FACIAL_L_Jawline_Rotation {type=Rotator}
 FACIAL_L_Jawline_Translation {type=Vector}
 FACIAL_L_UnderChin_Rotation {type=Rotator}
 FACIAL_L_UnderChin_Translation {type=Vector}
 FACIAL_R_Jawline_Rotation {type=Rotator}
 FACIAL_R_Jawline_Translation {type=Vector}
 FACIAL_R_UnderChin_Rotation {type=Rotator}
 FACIAL_R_UnderChin_Translation {type=Vector}
 HeadMesh {type=SkeletalMeshComponent}
 IsFaceBones {type=boolean}
 SkelMesh {type=SkeletalMeshComponent}
]
